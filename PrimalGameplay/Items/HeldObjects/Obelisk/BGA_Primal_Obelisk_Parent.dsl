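; Parent Blueprint for the Primal obelisk, which consumes a runestone held object and tosses out loot.
; Capsule overlap hides and expires the runestone, then starts a delayed loot timer with camera shakes, Niagara bursts, a dynamic-material emissive pulse and replicated gameplay cues.
; Sphere overlaps are counted on authority, and the count's OnRep runs a timer feeding the last player's runestone location to Niagara for proximity effects.
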(event EventBeginPlay
  (Audio|Components|Audio|FadeIn (Variables|Default|GetAmbientAudio) 0.2 1.0 0.0 "Linear" :!class /Script/Engine.AudioComponent)
  ;[1 1 1] Prepares Obelisk for changing the emissive later
  (bind _result (ArrayMacros|SafeArrayGet (Rendering|Material|GetMaterials (Variables|Default|GetStaticMesh)))
    (:%20%20
      (bind _returnvalue (Rendering|Material|CreateDynamicMaterialInstance _result "ObeliskMID" "None" :!class /Script/Engine.KismetMaterialLibrary))
      (bind _output_get (Variables|Default|SetObeliskMID _returnvalue))
      (Rendering|Material|SetMaterial (Variables|Default|GetStaticMesh) 0 _output_get)
      (bind _returnvalue_2 (Rendering|Material|GetVectorParameterValue _output_get "EmissiveColor" :!class /Script/Engine.MaterialInstanceDynamic))
      (Variables|Default|SetEmissiveColor _returnvalue_2))))

(event OnComponentBeginOverlap(Capsule)
  (params (OverlappedComponent :type Primitive%20Component%20Object%20Reference) (OtherActor :type Actor%20Object%20Reference) (OtherComp :type Primitive%20Component%20Object%20Reference) (OtherBodyIndex :type Integer) (bFromSweep :type Boolean) (SweepResult :type Hit%20Result%20Structure%20%28by%20ref%29))
  ;[1 1 1] On capsule overlap - check to see if it was a pawn. If it was, check if they have the specific held object. If it wasn't a pawn, check if it was the specific held object.
  (CallFunction|ResetLootSpawn)
  (bind _asfort_player_pawn (Utilities|Casting|CastToFortPlayerPawn OtherActor)
    (:then
      (bind _returnvalue (Fortnite|HeldObjects|GetHeldObjectAttachedtoPlayer _asfort_player_pawn))
      (bind _asb_held_object_primal_runestone_parent (Utilities|Casting|CastToB_HeldObject_Primal_Runestone_Parent _returnvalue))
      (Variables|/Game/Athena/Items/HeldObjects/HeldObjectVersion2/BGA_HeldObject_Physics_Parent.BGA_HeldObject_Physics_Parent_C|SetRepHideActor true _asb_held_object_primal_runestone_parent)
      (Actor|SetLifeSpan _asb_held_object_primal_runestone_parent 0.1))
    (:CastFailed
      (bind _asb_held_object_primal_runestone_parent_2 (Utilities|Casting|CastToB_HeldObject_Primal_Runestone_Parent OtherActor))
      ;[1 0.3496 0.9793] This is to fix the disparity between server and client locatiosn on the runestone. Occasionally the runestone will enter on the client but not on the server
      (bind _returnvalue_2 (Fortnite|HeldObjects|IsHeldObjectBeingHeld _asb_held_object_primal_runestone_parent_2))
      (if (not _returnvalue_2)
        (Variables|/Game/Athena/Items/HeldObjects/HeldObjectVersion2/BGA_HeldObject_Physics_Parent.BGA_HeldObject_Physics_Parent_C|SetRepHideActor true _asb_held_object_primal_runestone_parent_2)
        (Actor|SetLifeSpan _asb_held_object_primal_runestone_parent_2 0.1))))
  ;[1 1 1] Start the timer for spawning the loot and activate the relevant FX cues
  (bind _returnvalue_3 (Utilities|Time|SetTimerbyEvent (delegate Custom|Spawn&TossLoot) (Variables|Default|GetLootDelay)))
  (Variables|Default|SetTimer%20Handle _returnvalue_3)
  (Networking|CanExecuteCosmeticEvents
    (:True
      (Components|Activation|Activate (Variables|Default|GetLootCeremonyNiagaraTrigger) false)
      (Camera|PlayWorldCameraShake "/PrimalGameplay/Items/HeldObjects/Obelisk/B_Primal_Obelisk_LootWindUp_CameraShake.B_Primal_Obelisk_LootWindUp_CameraShake_C" (Transformation|GetWorldLocation (Variables|Default|GetStaticMesh)) 0.0 (Components|Sphere|GetScaledSphereRadius (Variables|Default|GetSphere)))
      (Utilities|FlowControl|Sequence
        (:then_0
          ;[1 1 1] Pulses the emissive on the obelisk between pickup slotted and loot dropped
          (bind (_direction _colormult) (Timeline|PulseEmissive PlayFromStart)
            (:Update
              (Rendering|Material|SetVectorParameterValue (Variables|Default|GetObeliskMID) "EmissiveColor" (Math|Color|LinearColor*Float (Variables|Default|GetEmissiveColor) _colormult) :!class /Script/Engine.MaterialInstanceDynamic))))
        (:then_1
          (Utilities|FlowControl|Delay (Variables|Default|GetLootDelay))
          (Components|Activation|Activate (Variables|Default|GetLootCeremonyNiagaraLootSpawnExplosion) true)
          (Camera|PlayWorldCameraShake "/PrimalGameplay/Items/HeldObjects/Obelisk/B_Primal_Obelisk_LootExplosion_CameraShake.B_Primal_Obelisk_LootExplosion_CameraShake_C" (Transformation|GetWorldLocation (Variables|Default|GetStaticMesh)) 0.0 (Components|Sphere|GetScaledSphereRadius (Variables|Default|GetSphere))))))
    (:False
      (Ability|ExecuteGameplayCueReplicated (Variables|Default|GetGCN_LootTrigger))
      (Utilities|FlowControl|Delay (Variables|Default|GetLootDelay))
      (Ability|ExecuteGameplayCueReplicated (Variables|Default|GetGCN_LootExplosion)))))

(event Custom|TriggerVisualEvents
  ;[1 1 1] Start the timer for spawning the loot and activate the relevant FX cues
  (Components|Activation|Activate (Variables|Default|GetLootCeremonyNiagaraTrigger) false)
  (Camera|PlayWorldCameraShake "/PrimalGameplay/Items/HeldObjects/Obelisk/B_Primal_Obelisk_LootWindUp_CameraShake.B_Primal_Obelisk_LootWindUp_CameraShake_C" (Transformation|GetWorldLocation (Variables|Default|GetStaticMesh)) 0.0 (Components|Sphere|GetScaledSphereRadius (Variables|Default|GetSphere)))
  (Utilities|FlowControl|Sequence
    (:then_0
      ;[1 1 1] Pulses the emissive on the obelisk between pickup slotted and loot dropped
      (Timeline|PulseEmissive PlayFromStart))
    (:then_1
      (Utilities|FlowControl|Delay (Variables|Default|GetLootDelay))
      (Components|Activation|Activate (Variables|Default|GetLootCeremonyNiagaraLootSpawnExplosion) true)
      (Camera|PlayWorldCameraShake "/PrimalGameplay/Items/HeldObjects/Obelisk/B_Primal_Obelisk_LootExplosion_CameraShake.B_Primal_Obelisk_LootExplosion_CameraShake_C" (Transformation|GetWorldLocation (Variables|Default|GetStaticMesh)) 0.0 (Components|Sphere|GetScaledSphereRadius (Variables|Default|GetSphere))))))

(event Custom|Spawn&TossLoot
  (bind _arrow (Variables|Default|GetArrow))
  (bind _returnvalue (Transformation|GetWorldLocation _arrow))
  ;[1 1 1] Spawn the loot and toss it, and call to deactivate the FX now that the ceremony is finished
  (Utilities|FlowControl|DoOnce false
    (:Completed
      (CallFunction|DeactivateFX)
      (bind _returnvalue_2 (Inventory|SpawnPickupInWorld (Utilities|ResolveSoftReference (Variables|Default|GetItemToSpawn)) 1 _returnvalue "0.000000,0.000000,1.000000" 0 false false true -1 "Tossed" "ItemSpawner" 0 false))
      (bind _output_get (Variables|Default|SetSpawnedPickup _returnvalue_2))
      (Pickup|TossPickup _output_get (+ _returnvalue (* (Transformation|GetForwardVector (Variables|Default|GetStaticMesh)) 5e+02)) 0 0 true true "Tossed" "Unset")
      (Networking|ForceNetUpdate)
      (Utilities|Time|ClearandInvalidateTimerbyHandle (Variables|Default|GetTimer%20Handle))
      (bind _spawnedpickup (Variables|Default|GetSpawnedPickup)
        (:then
          (Transformation|SetActorRotation _spawnedpickup (Math|Rotator|MakeRotator (.roll (Transformation|GetWorldRotation (Variables|Default|GetArrow))) (.pitch (Transformation|GetWorldRotation (Variables|Default|GetArrow))) (- (.yaw (Transformation|GetWorldRotation (Variables|Default|GetArrow))) 9e+01)) false))
        (:else)))))

(event Custom|ResetLootSpawn
  (bind _arrow (Variables|Default|GetArrow))
  (bind _returnvalue (Transformation|GetWorldLocation _arrow))
  ;[1 1 1] Spawn the loot and toss it, and call to deactivate the FX now that the ceremony is finished
  (Utilities|FlowControl|DoOnce false
    (:Completed
      (CallFunction|DeactivateFX)
      (bind _returnvalue_2 (Inventory|SpawnPickupInWorld (Utilities|ResolveSoftReference (Variables|Default|GetItemToSpawn)) 1 _returnvalue "0.000000,0.000000,1.000000" 0 false false true -1 "Tossed" "ItemSpawner" 0 false))
      (bind _output_get (Variables|Default|SetSpawnedPickup _returnvalue_2))
      (Pickup|TossPickup _output_get (+ _returnvalue (* (Transformation|GetForwardVector (Variables|Default|GetStaticMesh)) 5e+02)) 0 0 true true "Tossed" "Unset")
      (Networking|ForceNetUpdate)
      (Utilities|Time|ClearandInvalidateTimerbyHandle (Variables|Default|GetTimer%20Handle))
      (bind _spawnedpickup (Variables|Default|GetSpawnedPickup)
        (:then
          (Transformation|SetActorRotation _spawnedpickup (Math|Rotator|MakeRotator (.roll (Transformation|GetWorldRotation (Variables|Default|GetArrow))) (.pitch (Transformation|GetWorldRotation (Variables|Default|GetArrow))) (- (.yaw (Transformation|GetWorldRotation (Variables|Default|GetArrow))) 9e+01)) false))
        (:else)))))

(event Custom|DeactivateFX
  ;[1 1 1] Deacticvate the FX
  (Networking|CanExecuteCosmeticEvents
    (:True
      (Utilities|Time|ClearandInvalidateTimerbyHandle (Variables|Default|GetPlayerFXTrackingTimer))
      (Components|Activation|Deactivate (Variables|Default|GetLootCeremonyNiagaraTrigger)))))

(event Custom|StartPlayerTrackingTimer
  (bind _withinrangeniagara (Variables|Default|GetWithinRangeNiagara))
  ;[1 1 1] FX Work - Does some voodoo to track the relevant heldobject
  (bind _returnvalue (Utilities|Time|SetTimerbyEvent (delegate Custom|TickPlayerLocation) 0.066667 true))
  (Variables|Default|SetPlayerFXTrackingTimer _returnvalue)
  (Components|Activation|Activate _withinrangeniagara false)
  (Rendering|SetVisibility _withinrangeniagara true false)
  (Effects|Components|ParticleSystem|SetBoolParameter _withinrangeniagara "bIsStoneHeld" true)
  (Networking|SwitchHasAuthority
    (:Authority
      (Ability|ExecuteGameplayCueReplicated (Variables|Default|GetGCN_RunestoneWithinRange)))))

(event OnComponentBeginOverlap(Sphere)
  (params (OverlappedComponent :type Primitive%20Component%20Object%20Reference) (OtherActor :type Actor%20Object%20Reference) (OtherComp :type Primitive%20Component%20Object%20Reference) (OtherBodyIndex :type Integer) (bFromSweep :type Boolean) (SweepResult :type Hit%20Result%20Structure%20%28by%20ref%29))
  ;[1 1 1] Handle players entering and exiting the within-range sphere that triggers the tracking FX if the player is valid. Attempts to keep track of more than one relevant object should the situation arise where we have more than one within the sphere.
  (Networking|SwitchHasAuthority
    (:Authority
      (bind _asfort_player_pawn_athena (Utilities|Casting|CastToFortPlayerPawnAthena OtherActor)
        (:then
          (Variables|Default|SetLastPlayer _asfort_player_pawn_athena)
          (bind _returnvalue (Fortnite|HeldObjects|GetHeldObjectAttachedtoPlayer _asfort_player_pawn_athena))
          (Utilities|Casting|CastToB_HeldObject_Primal_Runestone_Parent _returnvalue))
        (:CastFailed
          (Utilities|Casting|CastToB_HeldObject_Primal_Runestone_Parent OtherActor)))
      (Variables|Default|SetNumOverlappingSpheres (+ (Variables|Default|GetNumOverlappingSpheres) 1)))))

(event Custom|TickPlayerLocation
  (bind _withinrangeniagara (Variables|Default|GetWithinRangeNiagara))
  ;[1 1 1] FX Work - Does some voodoo to track the relevant heldobject
  (bind _lastplayer (Variables|Default|GetLastPlayer)
    (:then
      (bind _returnvalue (Fortnite|HeldObjects|GetHeldObjectAttachedtoPlayer _lastplayer))
      (bind _asb_held_object_primal_runestone_parent (Utilities|Casting|CastToB_HeldObject_Primal_Runestone_Parent _returnvalue)
        (:then
          (Niagara|SetNiagaraVariable(Vector3) (Variables|Default|GetIdleNiagara) "PlayerLocation" (Transformation|GetActorLocation _asb_held_object_primal_runestone_parent))
          (Effects|Components|ParticleSystem|SetBoolParameter _withinrangeniagara "bIsStoneHeld" true))
        (:CastFailed
          (Effects|Components|ParticleSystem|SetBoolParameter _withinrangeniagara "bIsStoneHeld" false))))
    (:else)))

(event OnComponentEndOverlap(Sphere)
  (params (OverlappedComponent :type Primitive%20Component%20Object%20Reference) (OtherActor :type Actor%20Object%20Reference) (OtherComp :type Primitive%20Component%20Object%20Reference) (OtherBodyIndex :type Integer))
  ;[1 1 1] Handle players entering and exiting the within-range sphere that triggers the tracking FX if the player is valid. Attempts to keep track of more than one relevant object should the situation arise where we have more than one within the sphere.
  (Networking|SwitchHasAuthority
    (:Authority
      (bind _asfort_player_pawn_athena (Utilities|Casting|CastToFortPlayerPawnAthena OtherActor)
        (:then
          (bind _returnvalue (Fortnite|HeldObjects|GetHeldObjectAttachedtoPlayer _asfort_player_pawn_athena))
          (Utilities|Casting|CastToB_HeldObject_Primal_Runestone_Parent _returnvalue)
          (Variables|Default|SetLastPlayer 0))
        (:CastFailed
          (Utilities|Casting|CastToB_HeldObject_Primal_Runestone_Parent OtherActor)))
      (Variables|Default|SetNumOverlappingSpheres (- (Variables|Default|GetNumOverlappingSpheres) 1)))))

(fn BlueprintCanInteract
  (params (InteractingPawn :type Fort%20Pawn%20Object%20Reference) (InteractionBeingAttempted :type EInteractionBeingAttempted%20Enum :default "FirstInteraction") (InteractionType :type TInteractionType%20Enum :default "IT_NoInteraction")))

(fn BlueprintGetInteractionString
  (params (InteractingPawn :type Fort%20Pawn%20Object%20Reference) (InteractionBeingAttempted :type EInteractionBeingAttempted%20Enum :default "FirstInteraction")))

(fn OnRep_NumOverlappingSpheres
  (if (>= (Variables|Default|GetNumOverlappingSpheres) 1)
    (CallFunction|StartPlayerTrackingTimer)
    (else
      (Utilities|Time|ClearandInvalidateTimerbyHandle (Variables|Default|GetPlayerFXTrackingTimer)))))

(fn ConstructionScript)
